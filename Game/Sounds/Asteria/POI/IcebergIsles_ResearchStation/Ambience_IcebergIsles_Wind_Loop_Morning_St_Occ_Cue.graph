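# Morning wind ambience cue for the Iceberg Isles research station, using the Athena ambient sound class and attenuation.
# A quality-level node picks a quad mix for its first two slots and a stereo mix for the last two.
# Each mix weights an alpine mountain night wave over a looping, quieter UFO crash-site morning wave.
# Two nested, endlessly looping envelopers add slow pitch drift and a gentle volume swell of differing periods.
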SoundCue /Game/Sounds/Asteria/POI/IcebergIsles_ResearchStation/Ambience_IcebergIsles_Wind_Loop_Morning_St_Occ_Cue
root=Enveloper_6
nodes [
 {
  id=Enveloper_0
  children [QualityLevel_1]
  properties {
   bLoop=true
   bLoopIndefinitely=true
   loopEnd=40
   pitchCurve {editorCurveData {keys [
      {arriveTangent=0, interpMode=RCIM_Cubic, leaveTangent=0, tangentMode=RCTM_User, time=0}
      {arriveTangent=0, interpMode=RCIM_Cubic, leaveTangent=0, tangentMode=RCTM_User, time=20, value=0.92}
      {arriveTangent=0.004, interpMode=RCIM_Cubic, leaveTangent=0.004, time=40}
     ]}}
   volumeCurve {editorCurveData {keys [<at 0 1>]}}
  }
 }
 {
  id=Enveloper_6
  children [Enveloper_0]
  properties {
   bLoop=true
   bLoopIndefinitely=true
   loopEnd=30
   pitchCurve {editorCurveData {keys [{time=0}]}}
   volumeCurve {editorCurveData {keys [
      <at 0 1 RCIM_Cubic>
      {arriveTangent=0, interpMode=RCIM_Cubic, leaveTangent=0, tangentMode=RCTM_User, time=15, value=0.9}
      <at 30 1 RCIM_Cubic>
     ]}}
  }
 }
 {id=Mixer_0, children [WavePlayer_4,WavePlayer_0], properties {inputVolume [1.5,0.5]}}
 {id=Mixer_1, children [WavePlayer_1,WavePlayer_5], properties {inputVolume [1.5,0.5]}}
 {id=QualityLevel_1, children [Mixer_0,Mixer_0,Mixer_1,Mixer_1]}
 {
  id=WavePlayer_0
  properties {bLooping=true, soundWaveAssetPtr=/Game/Sounds/Apollo/Biomes/UFO_Crash_Sites/AmbientLoops/AMB_UFO_Crashsite_Morning_Quad}
 }
 {
  id=WavePlayer_1
  properties {bLooping=true, soundWaveAssetPtr=/Game/Sounds/Apollo/Biomes/UFO_Crash_Sites/AmbientLoops/AMB_UFO_Crashsite_Morning_Stereo}
 }
 {id=WavePlayer_4, properties {soundWaveAssetPtr=/Game/Sounds/Artemis/Biomes/Mountains/Ambience_Mountains_Alpine_Night_Quad}}
 {id=WavePlayer_5, properties {soundWaveAssetPtr=/Game/Sounds/Artemis/Biomes/Mountains/Ambience_Mountains_Alpine_Night_Stereo}}
]
properties {
 attenuationSettings=&Fort_Ambient_Athena_Att
 maxDistance=2097152
 pitchMultiplier=0.9
 soundClassObject=&Fort_Ambient_Athena
 volumeMultiplier=0.301
 preEffectBusSends [{}]
}
